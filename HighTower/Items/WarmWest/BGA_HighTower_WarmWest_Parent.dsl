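; Parent Blueprint for the WarmWest launch pad, which launches players and vehicles entering its sphere along LaunchDirection.
; Scalable floats gate it; players need a clear line trace.
; Launches fire a gameplay cue and quest events, tracking actors until overlap end or timeout.
; Optionally, players get a delayed redeploy effect until parachuting or landing; construction draws the predicted arc as an editor spline.
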(event Custom|BindOnTimeoutToRemovePlayer
  ;[1 1 1] When player get stuck closed to launch pad
  (Default|BindEventtoSetTimeouttoRemoveLaunched (delegate Custom|TimeoutToRemovedLaunched)))

(event EventBeginPlay
  ;[1 1 1] begin play
  (if (CallFunction|IsHardEnabled)
    (if (CallFunction|IsSoftEnabled)
      (CallFunction|BindonPlayerLaunched)
      (CallFunction|BindonTimeouttoRemovePlayer)
      (else
        (Components|Activation|Deactivate (Variables|Default|GetLauncher_Audio_Loop))
        (Variables|WarmWestDynamic|SetbSoftDisabled true)))
    (else
      (Actor|DestroyActor))))

(event Custom|TimeoutToRemovedLaunched
  (params (Launched :type Actor%20Object%20Reference))
  ;[1 1 1] When player get stuck closed to launch pad
  (Utilities|Time|SetTimerbyEvent (delegate Custom|RemoveLaunched) 1.0))

(event Custom|RemoveLaunched
  ;[1 1 1] When player get stuck closed to launch pad
  (CallFunction|RemoveLaunchedActor (stale Custom|TimeoutToRemovedLaunched Launched)))

(event Custom|OnBeginParachuteMovementEvent_Event)

(event Custom|OnPawnLanded_Event
  (params (Hit :type Hit%20Result%20Structure%20%28by%20ref%29)))

(event OnComponentBeginOverlap(Sphere)
  (params (OverlappedComponent :type Primitive%20Component%20Object%20Reference) (OtherActor :type Actor%20Object%20Reference) (OtherComp :type Primitive%20Component%20Object%20Reference) (OtherBodyIndex :type Integer) (bFromSweep :type Boolean) (SweepResult :type Hit%20Result%20Structure%20%28by%20ref%29))
  (if (not (Variables|WarmWestDynamic|GetbSoftDisabled))
    (bind (_fortpawn _fortvehicle) (Utilities|PlayerorVehicle OtherActor)
      (:Player
        ;[1 1 1] Player
        (bind _bcanlaunch (CallFunction|CanLaunchPlayer _fortpawn))
        (if _bcanlaunch
          (bind _lauched (CallFunction|LaunchPlayer _fortpawn))
          (if _lauched
            (Default|CallOnPlayerHasBeenLaunched _fortpawn)
            (CallFunction|AddLaunchedActorandSpawnGCN _fortpawn)
            (CallFunction|SendPlayerLaunchedQuestEvent _fortpawn))))
      (:Vehicle
        ;[1 1 1] Vehicle
        (bind _lauched_2 (CallFunction|LaunchVehicle _fortvehicle))
        (if _lauched_2
          (CallFunction|AddLaunchedActorandSpawnGCN _fortvehicle)
          (CallFunction|SendVehicleLaunchedQuestEvent _fortvehicle))))))

(event OnComponentEndOverlap(Sphere)
  (params (OverlappedComponent :type Primitive%20Component%20Object%20Reference) (OtherActor :type Actor%20Object%20Reference) (OtherComp :type Primitive%20Component%20Object%20Reference) (OtherBodyIndex :type Integer))
  (if (not (Variables|WarmWestDynamic|GetbSoftDisabled))
    (CallFunction|RemoveLaunchedActor OtherActor)))

(event Custom|BindOnPlayerLaunched
  ;[1 1 1] Player launched, delay before skydiving active
  (if (CallFunction|IsPlayerDeployEnabled)
    (Default|BindEventtoOnPlayerHasBeenLaunched (delegate Custom|OnPlayerLaunched))))

(event Custom|OnPlayerLaunched
  (params (PlayerPawn :type Fort%20Pawn%20Object%20Reference))
  ;[1 1 1] Player launched, delay before skydiving active
  (bind _asplayer_pawn_athena (Utilities|Casting|CastToPlayerPawn_Athena PlayerPawn))
  (Utilities|FlowControl|Delay (ScalableFloat|GetValueatLevel (Variables|WarmWestEditable|GetRedeployTime)))
  (bind _returnvalue (GameplayEffects|ApplyGameplayEffectSpecToSelf (Ability|GetAbilitySystemComponent _asplayer_pawn_athena) (GameplayEffects|MakeOutgoingSpec :self (Ability|GetAbilitySystemComponent _asplayer_pawn_athena) :GameplayEffectClass (Variables|WarmWestDefaults|GetGE_RedeployEffect) :Level 0.0)))
  (Variables|WarmWestDynamic|SetRedeployEffectHandle _returnvalue)
  (EventDispatchers|AssignOnBeginParachuteMovementEvent _asplayer_pawn_athena (delegate Custom|OnBeginParachuteMovementEvent_Event_0))
  (EventDispatchers|AssignOnPawnLanded _asplayer_pawn_athena (delegate Custom|OnPawnLanded_Event_0)))

(event Custom|OnPawnLanded_Event_0
  (params (Hit :type Hit%20Result%20Structure%20%28by%20ref%29))
  ;[1 1 1] Player launched, delay before skydiving active
  (GameplayEffects|RemoveActiveGameplayEffect (Ability|GetAbilitySystemComponent (stale Custom|OnPlayerLaunched _asplayer_pawn_athena)) (Variables|WarmWestDynamic|GetRedeployEffectHandle) -1)
  (EventDispatchers|UnbindEventfromOnPawnLanded (stale Custom|OnPlayerLaunched _asplayer_pawn_athena) (delegate Custom|OnPawnLanded_Event_0))
  (EventDispatchers|UnbindEventfromOnBeginParachuteMovementEvent (stale Custom|OnPlayerLaunched _asplayer_pawn_athena) (delegate Custom|OnBeginParachuteMovementEvent_Event_0)))

(event Custom|OnBeginParachuteMovementEvent_Event_0
  ;[1 1 1] Player launched, delay before skydiving active
  (GameplayEffects|RemoveActiveEffectswithGrantedTags (Ability|GetAbilitySystemComponent (stale Custom|OnPlayerLaunched _asplayer_pawn_athena)) "(GameplayTags=((TagName=\"Gameplay.Status.ImmuneFallingDamage\")))")
  (GameplayEffects|RemoveActiveGameplayEffect (Ability|GetAbilitySystemComponent (stale Custom|OnPlayerLaunched _asplayer_pawn_athena)) (Variables|WarmWestDynamic|GetRedeployEffectHandle) -1)
  (EventDispatchers|UnbindEventfromOnPawnLanded (stale Custom|OnPlayerLaunched _asplayer_pawn_athena) (delegate Custom|OnPawnLanded_Event_0))
  (EventDispatchers|UnbindEventfromOnBeginParachuteMovementEvent (stale Custom|OnPlayerLaunched _asplayer_pawn_athena) (delegate Custom|OnBeginParachuteMovementEvent_Event_0)))

(fn LaunchVehicle
  (params (Vehicle :type Fort%20Athena%20Vehicle%20Object%20Reference))
  (bind _location (CallFunction|GetVehicleCenterOfMassLocationandMassValue Vehicle))
  (Audio|BroadcastAbilityImpactatLocation 0 Vehicle (+ _location (* (Math|Rotator|GetRotationXVector (Transformation|GetWorldRotation (Variables|Default|GetLaunchDirection))) (* (ScalableFloat|GetValueatLevel (Variables|WarmWestEditable|GetVelocity)) (ScalableFloat|GetValueatLevel (Variables|WarmWestEditable|GetVehicleVelocityModifier))))) 777.0)
  (return true))

(fn AddLaunchedActorAndSpawnGCN
  (params (Actor :type Actor%20Object%20Reference%20%28by%20ref%29))
  (Utilities|Array|Add (Variables|WarmWestDynamic|GetOverlappedActors) Actor)
  (Default|CallSetTimeouttoRemoveLaunched Actor)
  (if (Networking|IsServer)
    (Ability|ExecuteGameplayCueWithParamsReplicated (Variables|WarmWestDefaults|GetTag_GCN_HandLaunch) (Ability|GameplayCue|MakeGameplayCueParameters :NormalizedMagnitude 0.0 :RawMagnitude 0.0 :OriginalTag (Variables|WarmWestDefaults|GetTag_GCN_HandLaunch) :Location (Transformation|GetActorLocation) :Normal (Transformation|GetForwardVector (Variables|Default|GetLaunchDirection)) :GameplayEffectLevel 1 :AbilityLevel 1 :bReplicateLocationWhenUsingMinimalRepProxy false))))

(fn CanLaunch
  (params (ItemToFind :type Fort%20Pawn%20Object%20Reference%20%28by%20ref%29))
  (return (not (Utilities|Array|ContainsItem (Variables|WarmWestDynamic|GetOverlappedActors) ItemToFind))))

(fn CanLaunchPlayer
  (params (Player :type Fort%20Pawn%20Object%20Reference))
  (bind _player (Variables|GetPlayer))
  (Utilities|IsValid Player
    (:Is%20Valid
      ;[1 1 1] Ignore player getting launched and all attached actors to that player
      (bind _output_get (LocalVariables|SetIgnoreActors (Actor|GetAttachedActors _player true false)))
      (Utilities|Array|Add _output_get _player)
      (bind (_outhit _returnvalue) (Collision|LineTraceByChannel (Transformation|GetActorLocation) (+ (Transformation|GetActorLocation) (* (Math|Vector|Normalize (- (Transformation|GetActorLocation (Variables|GetPlayer)) (Transformation|GetActorLocation))) (ScalableFloat|GetValueatLevel (Variables|Default|GetActivationTraceDistance)))) "TraceTypeQuery8" false (LocalVariables|GetIgnoreActors) "ForDuration" true "(R=1.000000,G=0.000000,B=0.000000,A=1.000000)" "(R=0.000000,G=1.000000,B=0.000000,A=1.000000)" 1e+01))
      (return (not _returnvalue)))
    (:Is%20Not%20Valid
      (return false))))

(fn DestroyBuildingsinPath
  (params (Pawn :type Fort%20Pawn%20Object%20Reference))
  (bind (_outhits _returnvalue) (Collision|MultiCapsuleTraceByChannel (Transformation|GetWorldLocation (Variables|Default|GetLaunchDirection)) (+ (Transformation|GetWorldLocation (Variables|Default|GetLaunchDirection)) (* (Math|Rotator|GetRotationXVector (Transformation|GetWorldRotation (Variables|Default|GetLaunchDirection))) 512.0)) (Variables|/Script/Engine.CapsuleComponent|GetCapsuleRadius (Variables|/Script/Engine.Character|GetCapsuleComponent (Variables|GetPawn))) (Variables|/Script/Engine.CapsuleComponent|GetCapsuleHalfHeight (Variables|/Script/Engine.Character|GetCapsuleComponent (Variables|GetPawn))) "TraceTypeQuery5"))
  (if _returnvalue
    (for _array_element _outhits
      (bind _asbuilding_actor (Utilities|Casting|CastToBuildingActor (.HitActor (Collision|BreakHitResult _array_element))))
      (Gameplay|Die _asbuilding_actor "(TagName=\"Gameplay.Effect.InstantDeath\")" (Pawn|GetController (Variables|GetPawn)) self))))

(fn GetFortPlayerController
  (params (Pawn :type Pawn%20Object%20Reference))
  (bind _asfort_player_controller (Utilities|Casting|CastToFortPlayerController (Pawn|GetController Pawn))
    (:then
      (return _asfort_player_controller))
    (:CastFailed
      (return 0))))

(fn GetVehicleCenterOfMassLocationAndMassValue
  (params (Vehicle :type Fort%20Vehicle%20Interface%20Interface))
  (bind _returnvalue (Vehicle|GetActor Vehicle))
  (bind _asprimitive_component (Utilities|Casting|CastToPrimitiveComponent (Variables|Transformation|GetRootComponent _returnvalue))
    (:then
      (bind (_returnvalue_x _returnvalue_y _returnvalue_z) (Math|Vector|BreakVector (Physics|GetCenterOfMass _asprimitive_component (Variables|WarmWestDefaultsVehicle|GetVehicleCenterOfMassSocketName))))
      (return _returnvalue_x _returnvalue_y (- _returnvalue_z (Variables|WarmWestDefaultsVehicle|GetVehicleLaunchFromCenterOfMassZOffset)) (Physics|GetMass _asprimitive_component)))
    (:CastFailed
      (return (Transformation|GetActorLocation _returnvalue) (Variables|WarmWestDefaultsVehicle|GetVehicleDefaultMassValue)))))

(fn IsHardEnabled
  (return (>= (ScalableFloat|GetValueatLevel (Variables|WarmWestEditable|GetHardEnabled)) 1.0)))

(fn IsPlayerDeployEnabled
  (return (>= (ScalableFloat|GetValueatLevel (Variables|WarmWestEditable|GetPlayerDeployEnabled)) 1.0)))

(fn IsSoftEnabled
  (return (>= (ScalableFloat|GetValueatLevel (Variables|WarmWestEditable|GetSoftEnabled)) 1.0)))

(fn JustPlayer
  (params (Actor :type Actor%20Object%20Reference))
  (if (not (CallFunction|PlayerinVehicleorJustVehicle Actor))
    (bind _asfort_pawn (Utilities|Casting|CastToFortPawn Actor)
      (:then
        (return _asfort_pawn true))
      (:CastFailed
        _)))
  (return 0 false))

(fn PlayerInVehicleOrJustVehicle
  (params (Actor :type Actor%20Object%20Reference))
  (bind _asfort_pawn (Utilities|Casting|CastToFortPawn Actor)
    (:then
      (bind _vehicle_occupied (Utilities|IsinVehicle _asfort_pawn)
        (:true
          (bind _asfort_athena_vehicle (Utilities|Casting|CastToFortAthenaVehicle _vehicle_occupied)
            (:then
              (return true _asfort_athena_vehicle))
            (:CastFailed
              (return false 0))))
        (:false
          _)))
    (:CastFailed
      _))
  (bind _asfort_athena_vehicle_2 (Utilities|Casting|CastToFortAthenaVehicle Actor)
    (:then
      (return true _asfort_athena_vehicle_2))
    (:CastFailed
      (return false 0))))

(fn RemoveLaunchedActor
  (params (Launched :type Actor%20Object%20Reference%20%28by%20ref%29))
  (if (Utilities|Array|ContainsItem (Variables|WarmWestDynamic|GetOverlappedActors) Launched)
    (Utilities|Array|RemoveItem (Variables|WarmWestDynamic|GetOverlappedActors) Launched)))

(fn SendPlayerLaunchedQuestEvent
  (params (FortPlayerPawn :type Pawn%20Object%20Reference))
  (bind _controller (CallFunction|GetFortPlayerController FortPlayerPawn))
  (Quest|SendComplexCustomStatEvent (Quests|GetQuestManager _controller "Invalid") 0 "(GameplayTags=((TagName=\"Athena.Quests.WarmWest.Launch\")))" "(GameplayTags=)" 1))

(fn SendVehicleLaunchedQuestEvent
  (params (Vehicle :type Fort%20Athena%20Vehicle%20Object%20Reference))
  (for _array_element (Vehicle|GetAllPassengers Vehicle)
    (Quest|SendComplexCustomStatEvent (Quests|GetQuestManager (Fortnite|GetFortPlayerControllerfromActor _array_element) "Invalid") 0 "(GameplayTags=((TagName=\"Athena.Quests.WarmWest.Launch\")))" "(GameplayTags=)" 1)))

(fn ConstructionScript
  (if (Editor|InEditororPIE)
    (bind (_outhit _outpathpositions _outlasttracedestination _returnvalue) (Game|PredictProjectilePathByTraceChannel (Transformation|GetWorldLocation (Variables|Default|GetLaunchDirection)) (* (Math|Rotator|GetRotationXVector (Transformation|GetWorldRotation (Variables|Default|GetLaunchDirection))) (ScalableFloat|GetValueatLevel (Variables|WarmWestEditable|GetVelocity))) true 16.0 "ECC_GameTraceChannel11" false (Utilities|Array|MakeArray self (Variables|WarmWestEditable|GetSentinalParentActor)) "None" 0.0 15.0 3e+01 (* (LocalVariables|GetDefaultGravity) (ScalableFloat|GetValueatLevel (Variables|WarmWestEditable|GetGravityScale)))))
    (if _returnvalue
      (Spline|SetSplinePoints (Variables|Default|GetEDITOR_VisualPath) _outpathpositions "World" true))))
